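# EQS query for the Deimos spawner that picks a nearby position around the querier, apparently to adjust player placement.
# A spiral donut generator creates rings of candidate points, projected onto geometry with capsule traces on a custom game channel.
# A capsule overlap test against WorldStatic, offset upward, filters and scores the points.
# An inverse-linear 2D distance score favours points closest to the querier.
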EnvQuery /Figment_Deimos/Spawners/EQS_DeimosSpawner_PlayerAdjustment
options [{
  generator {
   name=EnvQueryGenerator_Donut_0
   class=EnvQueryGenerator_Donut
   properties {
    bUseSpiralPattern=true
    center=&EnvQueryContext_Querier
    innerRadius=400
    numberOfRings=3
    outerRadius=700
    pointsPerRing=10
    projectionData {
     extentX=50
     extentZ=100
     navigationFilter=null
     postProjectionVerticalOffset=0
     serializedChannel=ECC_GameTraceChannel11
     traceChannel=TraceTypeQuery10
     traceMode=GeometryByChannel
     traceShape=Capsule
    }
   }
  }
  tests [
   EnvQueryTest_Overlap_0 {
    class=EnvQueryTest_Overlap
    properties {
     testPurpose=FilterAndScore
     overlapData {extentX=50, extentY=50, extentZ=100, overlapChannel=ECC_WorldStatic, overlapShape=Capsule, shapeOffset=<0 0 100>}
    }
   }
   EnvQueryTest_Distance_0 {
    class=EnvQueryTest_Distance
    properties {
     distanceTo=&EnvQueryContext_Querier
     filterType=Range
     floatValueMin=0
     scoringEquation=InverseLinear
     scoringFactor=1
     testComment=""
     testMode=Distance2D
     testPurpose=Score
    }
   }
  ]
 }]
